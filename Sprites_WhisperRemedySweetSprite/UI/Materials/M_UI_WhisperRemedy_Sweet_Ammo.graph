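# Translucent UI material for the Whisper Remedy Sweet ammo meter, drawing a vertical colored bar driven by MeterProgress.
# A box SDF function and rotated-UV smoothsteps build a tall, narrow shape with cut corners, plus a shadow mask.
# A step on inverted vertical UV against clamped progress forms the fill mask.
# Masks blend black to BarColor and combine with shadow, backplate and overall opacity parameters for transparency.
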Material /Sprites_WhisperRemedySweetSprite/UI/Materials/M_UI_WhisperRemedy_Sweet_Ammo
graph {
 properties {bAutomaticallySetUsageInEditor=false, blendMode=BLEND_Translucent, materialDomain=MD_UI, usedShadingModels=Default Lit}
 nodes [
  LinearInterpolate(
   A=Constant3Vector() {constant=<rgba 0 0 0 1>}
   B=VectorParameter() {parameterName=BarColor, defaultValue=<rgba 0.6376 0.0232 0.0307 1>}.RGB
   Alpha=Multiply(A=NamedRerouteUsage(), B=NamedRerouteUsage())
  )
  n1 = NamedRerouteUsage()
  n2 = NamedRerouteUsage()
  n3 = ScalarParameter() {defaultValue=0.2, parameterName=ShadowOpacity}
  n4 = Add(A=NamedRerouteUsage(), B=ScalarParameter() {defaultValue=0.5, parameterName=BackplateOpacity})
  Multiply(A=Min(A=LinearInterpolate(A=@n1, B=@n2, Alpha=@n3), B=Saturate(None=@n4)), B=ScalarParameter() {defaultValue=0.9, parameterName=Opacity}) {constB=0.8}
  n5 = MaterialFunctionCall() {materialFunction=&GetUserInterfaceUV}
  n6 = MaterialFunctionCall(UVs=@n5.Normalized UV, Size=Constant2Vector() {g=0.9, r=0.11}) {materialFunction=&MF_UI_Box}
  NamedRerouteDeclaration(None=@n6.SDF) {name=BoxSDF, nodeColor=<rgba 0 0 0 1>}
  n7 = ComponentMask(None=MaterialFunctionCall(UVs=@n5.Normalized UV, "Rotation Angle (0-1)"=Constant() {r=0.36}) {materialFunction=&CustomRotator}.Rotated Values) {r=true}
  n8 = Subtract(
   A=@n6.Fill
   B=SmoothStep(Value=NamedRerouteDeclaration(None=@n7) {name=Corner02, nodeColor=<rgba 0 0 0 1>}) {constMax=0.205, constMin=0.21, constValue=0.2}
  )
  n9 = MaterialFunctionCall(UVs=@n5.Normalized UV, "Rotation Angle (0-1)"=Constant() {r=0.74}) {materialFunction=&CustomRotator}
  n10 = SmoothStep(Value=NamedRerouteDeclaration(None=ComponentMask(None=@n9.Rotated Values) {r=true}) {name=Corner01, nodeColor=<rgba 0 0 0 1>}) {constMax=0.18, constMin=0.185}
  NamedRerouteDeclaration(None=Saturate(None=Subtract(A=@n8, B=@n10))) {name=Shape, nodeColor=<rgba 0 0 0 1>}
  n11 = OneMinus(None=NamedRerouteUsage())
  n12 = Subtract(
   A=SmoothStep(Value=Add(A=@n11) {constB=-0.43}) {constMax=0.55, constMin=0.5}
   B=SmoothStep(Value=NamedRerouteUsage()) {constMax=0.15, constMin=0.18}
  )
  n13 = SmoothStep(Value=NamedRerouteUsage()) {constMax=0.17, constMin=0.2, constValue=0.2}
  NamedRerouteDeclaration(None=Saturate(None=Subtract(A=@n12, B=@n13))) {name=Shadow, nodeColor=<rgba 0 0 0 1>}
  n14 = ComponentMask(None=MaterialFunctionCall() {materialFunction=&GetUserInterfaceUV}.Normalized UV) {g=true}
  n15 = ScalarParameter() {defaultValue=1, group=Progress Parameters, parameterName=MeterProgress, sliderMax=1, sortPriority=1}
  NamedRerouteDeclaration(None=Step(Y=OneMinus(None=@n14), X=Clamp(None=@n15) {maxDefault=0.825, minDefault=0.07}) {constX=0.5}) {name=MeterProgressMask, nodeColor=<rgba 0 0 0 1>}
 ]
}
