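# Translucent UI material for the Whisper Remedy Sweet Sprite center reticle, drawing mirrored brackets, heat bars and a center dot with shadow.
# Bracket and heat bar shapes come from MSDF textures sampled through symmetry UVs offset by SpreadOffset, while the dot is a procedural circle.
# PercentFill is remapped into a vertical step mask that splits filled and drained areas.
# The inUse, isCooldown and isDisabled scalars blend between state-specific colors and opacity masks.
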Material /Sprites_WhisperRemedySweetSprite/UI/Materials/M_UI_WhisperRemedy_Sweet_CenterReticle
graph {
 properties {bAutomaticallySetUsageInEditor=false, blendMode=BLEND_Translucent, materialDomain=MD_UI, usedShadingModels=Default Lit}
 nodes [
  n1 = LinearInterpolate(
   A=LinearInterpolate(B=VectorParameter() {parameterName=HeatColor, defaultValue=<rgba 0.6376 0.0296 0.0232 1>}.RGB, Alpha=NamedRerouteUsage()) {constA=1, constB=0}
   Alpha=OneMinus(None=NamedRerouteUsage())
  ) {constB=0}
  n2 = LinearInterpolate(A=@n1, B=VectorParameter() {parameterName=Color, defaultValue=<rgba 1 1 1 1>}.RGB, Alpha=NamedRerouteUsage())
  LinearInterpolate(A=@n2, B=@n1, Alpha=NamedRerouteUsage())
  LinearInterpolate(
   A=LinearInterpolate(A=NamedRerouteUsage(), B=NamedRerouteUsage(), Alpha=ScalarParameter() {parameterName=inUse})
   B=NamedRerouteUsage()
   Alpha=ScalarParameter() {parameterName=isDisabled}
  )
  n3 = LinearInterpolate(B=NamedRerouteUsage(), Alpha=ScalarParameter() {parameterName=isDisabled}) {constA=1, constB=0.2}
  n4 = MaterialFunctionCall!(MSDF RGB=TextureSample(UVs=NamedRerouteUsage()) {samplerType=SAMPLERTYPE_LinearColor, texture=/Sprites_WhisperRemedySweetSprite/UI/Textures/T_UI_WRSS_Bracket}.RGB) {materialFunction=/FortUILibrary/MaterialFunctions/Shapes/Modifiers/MF_UI_MSDFtoSDF}
  n5 = Max(
   A=MaterialFunctionCall(DistanceField=@n4.SDF, Threshold=Constant() {r=0.65}) {materialFunction=&MF_UI_AntiAliasing}.Outer
   B=NamedRerouteUsage()
  )
  n6 = MaterialFunctionCall(DistanceField=@n4.SDF, Threshold=Constant() {r=0.55}) {materialFunction=&MF_UI_AntiAliasing}
  n7 = Constant() {r=0.8}
  n8 = Max(
   A=Max(
    A=NamedRerouteDeclaration(None=@n5) {name=OuterBracketMask, nodeColor=<rgba 0 0 0 1>}
    B=LinearInterpolate(A=@n6.Outer, Alpha=@n7) {constB=0}
   )
   B=NamedRerouteUsage()
  )
  n9 = MaterialFunctionCall(MSDF RGB=TextureSample(UVs=NamedRerouteUsage()) {samplerType=SAMPLERTYPE_LinearColor, texture=/Sprites_WhisperRemedySweetSprite/UI/Textures/T_UI_WRSS_HeatBar}.RGB) {materialFunction=&MF_UI_MSDFtoSDF}
  n10 = NamedRerouteDeclaration(None=MaterialFunctionCall(DistanceField=@n9.SDF, Threshold=Constant() {r=0.55}) {materialFunction=&MF_UI_AntiAliasing}.Outer) {name=HeatbarMask, nodeColor=<rgba 0 0 0 1>}
  n11 = ComponentMask(None=TextureCoordinate()) {g=true}
  n12 = Saturate(None=Multiply(A=OneMinus(None=@n11)) {constB=2})
  n13 = LinearInterpolate(
   A=LinearInterpolate(A=Multiply(A=@n10, B=@n12), B=NamedRerouteUsage(), Alpha=ScalarParameter() {parameterName=inUse})
   B=Saturate(None=NamedRerouteUsage())
   Alpha=ScalarParameter() {parameterName=isCooldown}
  )
  n14 = Saturate(None=Subtract(A=NamedRerouteUsage()) {constB=0.9})
  n15 = ScalarParameter() {parameterName=isDisabled}
  Multiply(A=Max(A=Min(A=@n3, B=@n8), B=LinearInterpolate(A=@n13, B=@n14, Alpha=@n15)), B=ScalarParameter() {defaultValue=0.9, parameterName=Opacity}) {constB=0.9}
  n16 = MaterialFunctionCall!(UV, Offset=Constant() {r=-0.25}, Rotation, Scale=ScalarParameter() {defaultValue=1.02, parameterName=BracketHeight}, MirrorY) {materialFunction=/FortUILibrary/MaterialFunctions/Basics/MF_UI_Symmetry_2x1_DEPRECATED}
  n17 = Multiply(A=ComponentMask(None=@n16.Result) {r=true})
  n18 = Add(A=ScalarParameter() {defaultValue=0.05, parameterName=SpreadOffset}) {constB=-0.3}
  n19 = ComponentMask(None=@n16.Result) {g=true}
  NamedRerouteDeclaration(None=AppendVector(A=Subtract(A=@n17, B=@n18) {constB=0.5}, B=Multiply(A=@n19) {constB=1.8})) {name=BracketUVs, nodeColor=<rgba 0 0 0 1>}
  n20 = MaterialFunctionCall(Size=ScalarParameter() {defaultValue=0.035, parameterName=CircleRadius}) {materialFunction=&MF_UI_Circle}
  n21 = MaterialFunctionCall(DistanceField=@n20.SDF, Threshold=Constant() {r=0.001}) {materialFunction=&MF_UI_AntiAliasing}
  n22 = MaterialFunctionCall(DistanceField=@n20.SDF, Threshold=Constant() {r=0.02}) {materialFunction=&MF_UI_AntiAliasing}
  NamedRerouteDeclaration(None=Max(A=NamedRerouteDeclaration(None=@n21.Inner) {name=CenterDot, nodeColor=<rgba 1 0.4219 0 1>}, B=Multiply(A=@n22.Inner) {constB=0.2})) {name=CenterDotShadow, nodeColor=<rgba 1 0.0469 0 1>}
  n23 = MaterialFunctionCall(Offset=Constant() {r=-0.25}, Scale=ScalarParameter() {defaultValue=1.025, parameterName=HeatBarBracketHeight}) {materialFunction=&MF_UI_Symmetry_2x1_DEPRECATED}
  n24 = Multiply(A=ComponentMask(None=@n23.Result) {r=true})
  n25 = Add(A=ScalarParameter() {defaultValue=0.05, parameterName=SpreadOffset}) {constB=-0.25}
  n26 = ComponentMask(None=@n23.Result) {g=true}
  NamedRerouteDeclaration(None=AppendVector(A=Subtract(A=@n24, B=@n25) {constB=0.5}, B=Multiply(A=@n26) {constB=1.8})) {name=HeatbarBraketUV, nodeColor=<rgba 0 0 0 1>}
  NamedRerouteDeclaration(None=LinearInterpolate(A=@n2, B=VectorParameter() {parameterName=HeatbarOuterColor, defaultValue=<rgba 1 1 1 1>}.RGB, Alpha=NamedRerouteUsage())) {name=IdleAndCooldownColor, nodeColor=<rgba 0 0 0 1>}
  n27 = LinearInterpolate(
   A=LinearInterpolate(
    A=VectorParameter() {parameterName=ShadowColor, defaultValue=<rgba 0 0 0 1>}.RGB
    B=VectorParameter() {parameterName=Color, defaultValue=<rgba 1 1 1 1>}.RGB
    Alpha=NamedRerouteUsage()
   )
   B=VectorParameter() {parameterName=HeatColor, defaultValue=<rgba 0.6376 0.0296 0.0232 1>}.RGB
   Alpha=NamedRerouteUsage()
  )
  n28 = NamedRerouteUsage()
  NamedRerouteDeclaration(None=LinearInterpolate(
    A=LinearInterpolate(A=@n27, Alpha=@n28)
    B=VectorParameter() {parameterName=HeatbarOuterColor, defaultValue=<rgba 1 1 1 1>}.RGB
    Alpha=NamedRerouteUsage()
   )) {name=InUseColor, nodeColor=<rgba 0 0 0 1>}
  NamedRerouteDeclaration(None=MaterialFunctionCall(DistanceField=@n9.SDF, Threshold=Constant() {r=0.75}) {materialFunction=&MF_UI_AntiAliasing}.Outer) {name=InnerHeatMask, nodeColor=<rgba 0 0 0 1>}
  n29 = MaterialFunctionCall() {materialFunction=&GetUserInterfaceUV}
  n30 = Step(
   Y=OneMinus(None=ComponentMask(None=@n29.Normalized UV) {g=true})
   X=MaterialFunctionCall(
    Input=ScalarParameter() {defaultValue=0.5, parameterName=PercentFill}
    Input Low=Constant()
    Input High=Constant() {r=1}
    Target Low=Constant() {r=0.3}
    Target High=Constant() {r=0.66}
   ) {materialFunction=&RemapValueRange}.Result
  )
  NamedRerouteDeclaration(None=Min(A=@n30, B=NamedRerouteUsage())) {name=BracketInnerColorAlpha, nodeColor=<rgba 0 0 0 1>}
  n31 = Saturate(None=Subtract(A=ComponentMask(None=@n9.SDF) {g=true, r=true}) {constB=0.2})
  NamedRerouteDeclaration(None=Saturate(None=Multiply(A=@n31) {constB=2})) {name=HeatGlow, nodeColor=<rgba 0 0 0 1>}
  n32 = ComponentMask(None=TextureCoordinate()) {r=true}
  n33 = Max(A=Step(Y=@n32) {constX=0.45}, B=Step(Y=OneMinus(None=@n32)) {constX=0.45})
  NamedRerouteDeclaration(None=OneMinus(None=Subtract(A=@n33) {constB=0.1})) {name=Center Dot Mask, nodeColor=<rgba 0 0 0 1>}
  n34 = NamedRerouteUsage()
  n35 = Max(A=Multiply(A=@n34) {constB=0.1}, B=NamedRerouteUsage())
  NamedRerouteDeclaration(None=Min(A=@n35, B=OneMinus(None=@n30))) {name=DrainedMask, nodeColor=<rgba 0 0 0 1>}
  n36 = NamedRerouteDeclaration(None=Min(A=@n34, B=@n30)) {name=BracketMainColorAlpha, nodeColor=<rgba 0 0 0 1>}
  n37 = Max(A=@n35, B=@n36)
  NamedRerouteDeclaration(None=@n37) {name=CooldownColorProgressMask, nodeColor=<rgba 0 0 0 1>}
  NamedRerouteDeclaration(None=@n30) {name=ProgressMask, nodeColor=<rgba 0 0 0 1>}
  NamedRerouteDeclaration(None=LinearInterpolate(
    A=VectorParameter() {parameterName=ShadowColor, defaultValue=<rgba 0 0 0 1>}.RGB
    B=VectorParameter() {parameterName=Color, defaultValue=<rgba 1 1 1 1>}.RGB
    Alpha=NamedRerouteUsage()
   )) {name=DisabledColor, nodeColor=<rgba 0 0 0 1>}
  n38 = MaterialFunctionCall(Offset=Constant() {r=-0.245}, Scale=ScalarParameter() {defaultValue=1, parameterName=HeatBarHeight}) {materialFunction=&MF_UI_Symmetry_2x1_DEPRECATED}
  n39 = Multiply(A=ComponentMask(None=@n38.Result) {r=true})
  n40 = Add(A=ScalarParameter() {defaultValue=0.05, parameterName=SpreadOffset}) {constB=-0.23}
  n41 = ComponentMask(None=@n38.Result) {g=true}
  NamedRerouteDeclaration(None=AppendVector(A=Subtract(A=@n39, B=@n40) {constB=0.5}, B=Multiply(A=@n41) {constB=1.8})) {name=HeatbarUV, nodeColor=<rgba 0 0 0 1>}
  n42 = NamedRerouteUsage()
  n43 = MaterialFunctionCall(
   DistanceField=MaterialFunctionCall(MSDF RGB=TextureSample(UVs=@n42) {samplerType=SAMPLERTYPE_LinearColor, texture=/Sprites_WhisperRemedySweetSprite/UI/Textures/T_UI_WRSS_HeatBarOutline}.RGB) {materialFunction=&MF_UI_MSDFtoSDF}.SDF
   Threshold=Constant() {r=0.45}
  ) {materialFunction=&MF_UI_AntiAliasing}
  NamedRerouteDeclaration(None=LinearInterpolate(A=Multiply(A=@n43.Outer) {constB=0.5}, Alpha=ScalarParameter() {parameterName=isCooldown}) {constB=0}) {name=HeatBarInnerMask, nodeColor=<rgba 0 0 0 1>}
  n44 = Max(A=@n36, B=Min(A=@n30, B=NamedRerouteUsage()))
  NamedRerouteDeclaration(None=Max(A=@n37, B=NamedRerouteDeclaration(None=@n44) {name=BracketGlowColorAlpha, nodeColor=<rgba 0 0 0 1>})) {name=inUseBracketFinalOpacity, nodeColor=<rgba 0 0 0 1>}
  NamedRerouteUsage()
  ScalarParameter() {defaultValue=0.05, parameterName=SpreadOffset}
  VectorParameter() {parameterName=ShadowColor, defaultValue=<rgba 0 0 0 1>}
  VectorParameter() {parameterName=HeatColor, defaultValue=<rgba 0.6376 0.0296 0.0232 1>}
 ]
}
