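# Reticle widget for the Whisper Remedy Sweet held laser weapon, showing overheat and laser state through one center-image dynamic material.
# On construct it caches the equipped weapon, binds its overheat and laser-state events and applies current values.
# PercentFill is one minus overheat over a curve-table maximum, while state toggles inUse, isCooldown and isDisabled, and zero overheat clears cooldown.
# Targeting changes set SpreadOffset directly; the graph never plays Anim_ADS.
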WidgetBlueprint /Sprites_WhisperRemedySweetSprite/UI/WBP_CherrySmokeWhisperRemedy_Sweet_Reticle
animations [Anim_ADS {
  binding_count=1
  length_seconds=0.25
  object=&self:Anim_ADS
  sequence {bindings [Image_Center {
     id=#1
     possessed_class=Image
     tracks [{
       type=MovieSceneWidgetMaterialTrack
       sections [{
         type=MovieSceneParameterSection
         end_frame=5
         start_frame=0
         channels [SpreadOffset {type=MovieSceneScriptingFloatChannel, keys [[0,1],[2,1.1],[5,0.8]]}]
        }]
      }]
    }]}
 }]
graphs [
 EventGraph {nodes [
   n1 = UpdateOverheat!() -> (OutputDelegate, then, NewOverheat:double=~) {is=title K2Node_CustomEvent}
   > n2 = Branch(Condition="Nearly Equal (Float)"(A=@n1.NewOverheat, B=0, ErrorTolerance=0))
   > n3 = SetScalarParameterValue(self=Get ImageCenterMaterial() -> (ImageCenterMaterial:Material Instance Dynamic Object Reference), ParameterName=isCooldown, Value=0)
   SetScalarParameterValue(
    execute=[@n3.then, @n2.else]
    self=Get ImageCenterMaterial()
    ParameterName=PercentFill
    Value=NormalizeOverheatFloat!(FloatToNormalize:double=@n1.NewOverheat) -> (NormalizedFloat:double)
   )
   n4 = DoOnLaserStateChanged!() -> (OutputDelegate, then, LaserState:Enum_WhisperRemedy_Sweet_LaserBeamStates Enum=NewEnumerator0) {is=title K2Node_CustomEvent}
   > n5 = Switch on Enum_WhisperRemedy_Sweet_LaserBeamStates!(execute, Selection:Enum_WhisperRemedy_Sweet_LaserBeamStates Enum=@n4.LaserState) -> (NewEnumerator0:Exec, NewEnumerator1:Exec, NewEnumerator2:Exec) {is=title K2Node_SwitchEnum}
   n6 = Get ImageCenterMaterial()
   > SetScalarParameterValue(self=@n6, ParameterName=inUse, Value=1)
   > SetScalarParameterValue(self=@n6, ParameterName=isCooldown, Value=0)
   > SetScalarParameterValue(self=@n6, ParameterName=isDisabled, Value=0)
   SetScalarParameterValue(execute=@n5.NewEnumerator1, self=@n6, ParameterName=inUse, Value=0)
   > SetScalarParameterValue(self=@n6, ParameterName=isCooldown, Value=1)
   > SetScalarParameterValue(self=@n6, ParameterName=isDisabled, Value=0)
   SetScalarParameterValue(execute=@n5.NewEnumerator2, self=@n6, ParameterName=inUse, Value=0)
   > SetScalarParameterValue(self=@n6, ParameterName=isCooldown, Value=0)
   > SetScalarParameterValue(self=@n6, ParameterName=isDisabled, Value=1)
   Event Construct()
   > Parent: Construct()
   > n7 = Cast To B_Sprite_Held_Weapon_WhisperRemedySweet!(
    execute
    Object=GetEquippedWeapon!(self:Fort Weapon Reticle Extension Widget Base Object Reference) -> (ReturnValue:Fort Weapon Object Reference)
   ) -> (then, CastFailed, AsB Sprite Held Weapon Whisper Remedy Sweet:B Sprite Held Weapon Whisper Remedy Sweet Object Reference) {is=title K2Node_DynamicCast}
   > n8 = Set EquippedWeapon(EquippedWeapon:B Sprite Held Weapon Whisper Remedy Sweet Object Reference=@n7) -> (Output_Get:B Sprite Held Weapon Whisper Remedy Sweet Object Reference)
   > Bind Event to On Overheat Changed(self:B Sprite Held Weapon Whisper Remedy Sweet Object Reference=@n8, Delegate=Create Event()) {is=title K2Node_AddDelegate}
   > Bind Event to On Laser State Changed(self:B Sprite Held Weapon Whisper Remedy Sweet Object Reference=@n8, Delegate=Create Event()) {is=title K2Node_AddDelegate}
   > n9 = GetDynamicMaterial(self=Get Image_Center() -> (Image_Center:Image Object Reference))
   > n10 = Set ImageCenterMaterial(ImageCenterMaterial:Material Instance Dynamic Object Reference=@n9) -> (Output_Get:Material Instance Dynamic Object Reference)
   n11 = Get EquippedWeapon() -> (EquippedWeapon:B Sprite Held Weapon Whisper Remedy Sweet Object Reference)
   > SetScalarParameterValue(
    self=@n10
    ParameterName=PercentFill
    Value=NormalizeOverheatFloat(FloatToNormalize=Get CurrentOverheat!(self:B Sprite Held Weapon Whisper Remedy Sweet Object Reference=@n11) -> (CurrentOverheat:double))
   )
   > DoOnLaserStateChanged!(
    execute
    self
    LaserState:Enum_WhisperRemedy_Sweet_LaserBeamStates Enum=Get CurrentLaserState!(self:B Sprite Held Weapon Whisper Remedy Sweet Object Reference=Get EquippedWeapon()) -> (CurrentLaserState:Enum_WhisperRemedy_Sweet_LaserBeamStates Enum=NewEnumerator0)
   ) -> (then) {is=title K2Node_CallFunction}
   Event Destruct()
   > Parent: Destruct()
   > n12 = Get!(execute:Exec) -> (then:Exec, else:Exec, EquippedWeapon:B Sprite Held Weapon Whisper Remedy Sweet Object Reference)
   > Unbind Event from On Laser State Changed(self:B Sprite Held Weapon Whisper Remedy Sweet Object Reference=@n12, Delegate=Create Event()) {is=title K2Node_RemoveDelegate}
   > Unbind Event from On Overheat Changed(self:B Sprite Held Weapon Whisper Remedy Sweet Object Reference=@n12, Delegate=Create Event()) {is=title K2Node_RemoveDelegate}
   n13 = Event OnTargetingChanged!() -> (OutputDelegate, then, bIsADS:bool) {is=title K2Node_Event}
   > n14 = Branch(Condition=@n13.bIsADS)
   > SetScalarParameterValue(self=Get ImageCenterMaterial(), ParameterName=SpreadOffset, Value=0.1)
   SetScalarParameterValue(execute=@n14.else, self=Get ImageCenterMaterial(), ParameterName=SpreadOffset, Value=0.05)
  ]}
 NormalizeOverheatFloat {nodes [
   n1 = NormalizeOverheatFloat!() -> (then, FloatToNormalize:double=~) {is=title K2Node_FunctionEntry}
   n2 = GetValueAtLevel!(Input:"Scalable Float Structure (by ref)"=Get MaxOverheatValue() -> (MaxOverheatValue:Scalable Float Structure), Level:float=0) -> (ReturnValue:float)
   > Return Node!(execute, NormalizedFloat:double=float - float(A=1, B=float / float(A=@n1, B=@n2)))
  ]}
]
variables [
 CooldownDelay {
  type=ScalableFloat
  default {
   registryType=None
   value=1
   curve {
    curveTable=/Sprites_WhisperRemedySweetSprite/DataTables/WhisperRemedySweetSpriteGameData
    rowName=WhisperRemedy.Sweet.Overheat.Cooldown.Duration
   }
  }
 }
 DecrementAmount {
  type=ScalableFloat
  default {
   registryType=None
   value=1
   curve {curveTable=&WhisperRemedySweetSpriteGameData, rowName=WhisperRemedy.Sweet.Overheat.Decrement}
  }
 }
 DecrementDelay {
  type=ScalableFloat
  default {
   registryType=None
   value=1
   curve {
    curveTable=&WhisperRemedySweetSpriteGameData
    rowName=WhisperRemedy.Sweet.Overheat.Decrement.Delay
   }
  }
 }
 DecrementLoop_TimerHandle {type=TimerHandle}
 DecrementSpeed {
  type=ScalableFloat
  default {
   registryType=None
   value=1
   curve {
    curveTable=&WhisperRemedySweetSpriteGameData
    rowName=WhisperRemedy.Sweet.Overheat.Decrement.Speed
   }
  }
 }
 EquippedWeapon {type=B_Sprite_Held_Weapon_WhisperRemedySweet_C, default=null}
 ImageCenterMaterial {type=MaterialInstanceDynamic}
 IncrementAmount {
  type=ScalableFloat
  default {
   registryType=None
   value=1
   curve {curveTable=&WhisperRemedySweetSpriteGameData, rowName=WhisperRemedy.Sweet.Overheat.Increment}
  }
 }
 IncrementLoop_TimerHandle {type=TimerHandle}
 IncrementSpeed {
  type=ScalableFloat
  default {
   registryType=None
   value=1
   curve {
    curveTable=&WhisperRemedySweetSpriteGameData
    rowName=WhisperRemedy.Sweet.Overheat.Increment.Speed
   }
  }
 }
 MaxOverheatValue {
  type=ScalableFloat
  default {
   registryType=None
   value=1
   curve {curveTable=&WhisperRemedySweetSpriteGameData, rowName=WhisperRemedy.Sweet.Overheat.Max}
  }
 }
]
widgets {
 info {
  inheritedWidgetCount=1
  parentClass=/Script/FortniteUI.FortWeaponReticleExtensionWidgetBase
  rootWidgetClass=&Overlay
  widgetCount=3
 }
 widgets [
  OverlayAll {widgetClassPath=&Overlay, widget=OverlayAll}
  Image_Center {
   widgetClassPath=&Image
   bIsVariable=true
   parent=OverlayAll
   widget=Image_Center
   properties {brush {
     bIsDynamicallyLoaded=false
     drawAs=Image
     imageType=FullColor
     mirroring=NoMirror
     resourceName=None
     resourceObject=/Sprites_WhisperRemedySweetSprite/UI/Materials/M_UI_WhisperRemedy_Sweet_CenterReticle
     tiling=NoTile
     imageSize=<110 110>
     margin=<ltrb 0 0 0 0>
     outlineSettings {
      bUseBrushTransparency=false
      roundingType=HalfHeightRadius
      width=0
      color {colorUseRule=UseColor_Specified, specifiedColor=<rgba 0 0 0 0>}
      cornerRadii=<0 0 0 0>
     }
     tintColor {colorUseRule=UseColor_Specified, specifiedColor=<rgba 1 1 1 1>}
     uVRegion {bIsValid=false, max=<0 0>, min=<0 0>}
    }}
   slot {class=OverlaySlot}
  }
  TutorialHighlight {widgetClassPath=&TutorialHighlight}
 ]
}
